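; Screen actor that shows a character static mesh as a flattened 2D image on a screen mesh, with a replicated banner texture.
; Construction Script applies mesh and material overrides, then creates dynamic materials feeding screen bend, flatten, bounds-based center, camera move/zoom and edge-blend parameters to the character's shaders.
; With the character enabled, a NoLogo screen mesh gets logo-disabled materials.
; OnRep_BannerTexture pushes the banner to the cached MIDs; event graph handlers are empty.
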(event EventBeginPlay)

(event Collision|EventActorBeginOverlap
  (params (OtherActor :type Actor%20Object%20Reference)))

(event EventTick
  (params (DeltaSeconds :type Float%20%28single-precision%29)))

(fn OnRep_BannerTexture
  (CallFunction|SetBannerTexture (Variables|Default|GetBannerTexture)))

(fn SetBannerTexture
  (params (Value :type Texture%20Object%20Reference))
  (Utilities|IsValid Value
    (:Is%20Valid
      (for _array_element (Variables|Default|GetMIDs)
        (Rendering|Material|SetTextureParameterValue _array_element "Banner" Value)))))

(fn ConstructionScript
  (bind _returnvalue (Transformation|GetActorRotation))
  (bind _screenmesh (Variables|Default|GetScreenMesh))
  (bind _charactermesh (Variables|Default|GetCharacterMesh))
  (Utilities|Array|Clear (Variables|Default|GetMIDs))
  (Utilities|FlowControl|Sequence
    (:then_1
      (Utilities|FlowControl|Sequence
        (:then_0
          (bind _charactermeshoverride (Variables|Default|GetCharacterMeshOverride)
            (:then
              (Components|StaticMesh|SetStaticMesh (Variables|Default|GetCharacterMesh) _charactermeshoverride))
            (:else)))
        (:then_1
          (for (_array_element _array_index) (Variables|Default|GetCharacterMaterialOverrides)
            (Rendering|Material|SetMaterial (Variables|Default|GetCharacterMesh) _array_index _array_element)))
        (:then_2
          (bind _screenmeshoverride (Variables|Default|GetScreenMeshOverride)
            (:then
              (Components|StaticMesh|SetStaticMesh (Variables|Default|GetScreenMesh) _screenmeshoverride))
            (:else)))
        (:then_3
          (for (_array_element_2 _array_index_2) (Variables|Default|GetScreenMaterialOverrides)
            (Rendering|Material|SetMaterial (Variables|Default|GetScreenMesh) _array_index_2 _array_element_2)))
        (:then_4
          (if (Variables|Default|GetCharacter%20Enabled)
            (Utilities|IsValid (Variables|/Script/Engine.StaticMeshComponent|GetStaticMesh (Variables|Default|GetCharacterMesh))
              (:Is%20Valid
                (Transformation|SetRelativeTransform (Variables|Default|GetCharacterMesh) (Math|Transform|MakeTransform (Math|Vector|MakeVector 0.0 (Variables|Default|GetCharacterOffsetY) (Variables|Default|GetCharacterOffsetZ)) (Math|Rotator|MakeRotator 0.0 0.0 (Variables|CameraMovement|GetCameraRotation)) (Math|Conversions|ToVector(Float) (Variables|Default|GetCharacterScale))) false false)
                (for (_array_element_3 _array_index_3) (Rendering|Material|GetMaterials _charactermesh)
                  (bind _returnvalue_2 (Rendering|Material|CreateDynamicMaterialInstance _charactermesh _array_index_3 _array_element_3 "None" :!class /Script/Engine.PrimitiveComponent))
                  (Rendering|Material|SetScalarParameterValue _returnvalue_2 "ScreenBend" (/ (Variables|Default|GetBend) 1e+02) :!class /Script/Engine.MaterialInstanceDynamic)
                  (Rendering|Material|SetVectorParameterValue _returnvalue_2 "Direction" (Math|Conversions|ToLinearColor(Vector) (Math|Vector|GetRightVector _returnvalue)) :!class /Script/Engine.MaterialInstanceDynamic)
                  (Rendering|Material|SetScalarParameterValue _returnvalue_2 "FlattenAmount" (Variables|Default|GetFlatten) :!class /Script/Engine.MaterialInstanceDynamic)
                  (Rendering|Material|SetVectorParameterValue _returnvalue_2 "ScreenCenter" (Math|Conversions|ToLinearColor(Vector) (+ (Transformation|GetWorldLocation (Variables|Default|GetCharacterMesh)) (* (Math|Vector|GetUpVector _returnvalue) (+ (Variables|Default|GetScreenCenter%20Z%20Offset) (.z (Math|BoxSphereBounds|BreakBoxSphereBounds (Fortnite|Utility|GetLocalBounds (Variables|Default|GetScreenMesh)))))))) :!class /Script/Engine.MaterialInstanceDynamic)
                  (Rendering|Material|SetScalarParameterValue _returnvalue_2 "ScreenSize_X" (Variables|Default|GetScreen_Width) :!class /Script/Engine.MaterialInstanceDynamic)
                  (Rendering|Material|SetScalarParameterValue _returnvalue_2 "ScreenSize_Y" (Variables|Default|GetScreen_Height) :!class /Script/Engine.MaterialInstanceDynamic)
                  (Rendering|Material|SetScalarParameterValue _returnvalue_2 "CamerMoveSpeed" (Variables|CameraMovement|GetCamerMoveSpeed) :!class /Script/Engine.MaterialInstanceDynamic)
                  (Rendering|Material|SetScalarParameterValue _returnvalue_2 "CamerZoomSpeed" (Variables|CameraMovement|GetCamerZoomSpeed) :!class /Script/Engine.MaterialInstanceDynamic)
                  (Rendering|Material|SetScalarParameterValue _returnvalue_2 "CameraMove_Pos_X_1" (Variables|CameraMovement|GetCameraPos_A) :!class /Script/Engine.MaterialInstanceDynamic)
                  (Rendering|Material|SetScalarParameterValue _returnvalue_2 "CameraMove_Pos_X_2" (Variables|CameraMovement|GetCameraPos_B) :!class /Script/Engine.MaterialInstanceDynamic)
                  (Rendering|Material|SetScalarParameterValue _returnvalue_2 "CameraMove_Zoom_X_1" (Variables|CameraMovement|GetCameraZoom_A) :!class /Script/Engine.MaterialInstanceDynamic)
                  (Rendering|Material|SetScalarParameterValue _returnvalue_2 "CameraMove_Zoom_X_2" (Variables|CameraMovement|GetCameraZoom_B) :!class /Script/Engine.MaterialInstanceDynamic)
                  (Rendering|Material|SetScalarParameterValue _returnvalue_2 "Falloff_Screen_Opacity" (Variables|CharacterScreenBlend|GetCharacterScreenBlend_Falloff) :!class /Script/Engine.MaterialInstanceDynamic)
                  (Rendering|Material|SetScalarParameterValue _returnvalue_2 "OpacityScreen_Mask" (- 1.0 (Variables|CharacterScreenBlend|GetCharacterScreenBlend_Opacity_Inner)) :!class /Script/Engine.MaterialInstanceDynamic)
                  (Rendering|Material|SetScalarParameterValue _returnvalue_2 "OpacityScreen_Mask2" (- 1.0 (Variables|CharacterScreenBlend|GetCharacterScreenBlend_Opacity_Edges)) :!class /Script/Engine.MaterialInstanceDynamic)
                  (Utilities|Array|Add (Variables|Default|GetMIDs) _returnvalue_2)))
              (:Is%20Not%20Valid
                _)))
          (Components|DestroyComponent (Variables|Default|GetCharacterMesh)))
        (:then_5
          (for (_array_element_4 _array_index_4) (Rendering|Material|GetMaterials _screenmesh)
            (bind _returnvalue_3 (Rendering|Material|CreateDynamicMaterialInstance _array_element_4 "None" "None" :!class /Script/Engine.KismetMaterialLibrary))
            (Rendering|Material|SetMaterial _screenmesh _array_index_4 _returnvalue_3)
            (Utilities|Array|Add (Variables|Default|GetMIDs) _returnvalue_3)))
        (:then_6
          (if (Variables|Default|GetCharacter%20Enabled)
            (for _array_element_5 (Variables|Default|GetScreenMaterialOverrides)
              (bind _returnvalue_4 (Rendering|Material|CreateDynamicMaterialInstance (Variables|Default|GetScreenMesh_NoLogo_High) 0 _array_element_5 "None" :!class /Script/Engine.PrimitiveComponent))
              (Rendering|Material|SetScalarParameterValue _returnvalue_4 "LogoEnabled" 0.0 :!class /Script/Engine.MaterialInstanceDynamic))
            (Transformation|SetRelativeScale3D (Variables|Default|GetScreenMesh) "1.000000,1.000000,1.000000")
            (else
              (Components|DestroyComponent (Variables|Default|GetScreenMesh_NoLogo_High)))))))))
