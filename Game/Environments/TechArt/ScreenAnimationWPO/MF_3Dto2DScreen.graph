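# Material function that squashes a 3D mesh into a bent 2D screen image via World Position Offset, behind a static switch.
# Vertices flatten along a direction vector, rotate about Z for curvature, then shift toward the camera with a depth-based push to retain layering.
# Opacity mask crops to screen bounds, and MPC_Hologram's ScreensEnabled at zero collapses vertices to the object pivot.
# MF_ScreenProgress receives the result with a falloff-based mask for an optional overlay.
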MaterialFunction /Game/Environments/TechArt/ScreenAnimationWPO/MF_3Dto2DScreen
graph {nodes [
  Constant()
  n1 = FunctionInput() {inputType=FunctionInput_MaterialAttributes}
  n2 = GetMaterialAttributes(None=@n1) {attributeGetTypes [#1, #2, #3]}
  n3 = ObjectPositionWS()
  n4 = WorldPosition()
  n5 = NamedRerouteUsage()
  n6 = NamedRerouteUsage()
  n7 = DotProduct(A=@n5, B=@n6)
  n8 = Multiply(A=@n6) {constB=-1}
  n9 = Add(A=NamedRerouteUsage(), B=Multiply(A=Multiply(A=@n7, B=@n8), B=ScalarParameter() {defaultValue=0.98, parameterName=FlattenAmount}))
  n10 = DotProduct(A=CrossProduct(A=NamedRerouteUsage(), B=Constant3Vector() {constant=<rgba 0 0 1 0>}), B=@n5)
  n11 = ScalarParameter() {defaultValue=158, parameterName=ScreenSize_X}
  n12 = Divide(A=@n10, B=@n11) {constB=200}
  n13 = ScalarParameter() {defaultValue=-0.042, parameterName=ScreenBend}
  n14 = StaticSwitchParameter(True=ObjectPositionWS(), False=VectorParameter() {parameterName=ScreenCenter, defaultValue=<rgba 0 0 0 1>}.RGB) {parameterName=UseLocalOrWorldProjection}
  n15 = WorldPosition()
  n16 = Add(
   A=@n9
   B=RotateAboutAxis(
    NormalizedRotationAxis=Constant3Vector() {constant=<rgba 0 0 1 0>}
    RotationAngle=Multiply(A=@n12, B=@n13)
    PivotPoint=@n14
    Position=Add(A=@n15.XYZ, B=@n9)
   )
  )
  n17 = WorldPosition()
  n18 = Normalize(VectorInput=Subtract(A=CameraPositionWS(), B=Add(A=@n17.XYZ, B=@n16)))
  n19 = ObjectRadius()
  n20 = Divide(A=Add(A=@n7, B=@n19) {constB=100}, B=Multiply(A=@n19) {constB=2}) {constB=200}
  n21 = ScalarParameter() {defaultValue=70, parameterName=CameraDirectionCharacterDepthPush}
  n22 = Add(A=@n16, B=Multiply(A=@n18, B=Multiply(A=@n20, B=@n21)))
  n23 = Multiply(A=@n18, B=ScalarParameter() {defaultValue=10, parameterName=CameraDirectionPush})
  n24 = Saturate(None=CollectionParameter() {collection=/Game/Environments/TechArt/Hologram/MPC_Hologram, parameterName=ScreensEnabled})
  n25 = SetMaterialAttributes!(
   MaterialAttributes=@n2.MaterialAttributes
   WorldPositionOffset=LinearInterpolate(A=Subtract(A=@n3, B=@n4.XYZ), B=Add(A=@n22, B=@n23), Alpha=Ceil(None=@n24))
   OpacityMask=Multiply(A=@n2.OpacityMask, B=NamedRerouteUsage())
  ) {attributeSetTypes [#1, #2]}
  n26 = DotProduct(A=NamedRerouteUsage(), B=NamedRerouteUsage())
  n27 = Saturate(None=Power(Base=OneMinus(None=@n26), Exp=ScalarParameter() {defaultValue=2, group=ScreenMask, parameterName=Falloff_Screen_Opacity}))
  n28 = LinearInterpolate(
   A=ScalarParameter() {defaultValue=0.25, group=ScreenMask, parameterName=OpacityScreen_Mask}
   B=ScalarParameter() {defaultValue=0.75, group=ScreenMask, parameterName=OpacityScreen_Mask2}
   Alpha=@n27
  )
  FunctionOutput(None=MaterialFunctionCall!(
    In=StaticSwitchParameter(True=@n25, False=@n1) {parameterName=Use3Dto2DScreenWPO}
    UVs=NamedRerouteUsage()
    Mask=Saturate(None=@n28)
    IsEnabled=StaticBoolParameter() {parameterName=UseScreenProgressOverlay}
   ) {materialFunction=/Game/Environments/TechArt/ScreenProgress/MF_ScreenProgress}.Result)
  Multiply(A=Time(), B=ScalarParameter() {defaultValue=0.3, parameterName=CamerZoomSpeed})
  n29 = Subtract(A=Add(A=NamedRerouteUsage(), B=WorldPosition().XYZ), B=@n14)
  NamedRerouteDeclaration(None=ShaderStageSwitch(PixelShader=VertexInterpolator(VS=@n29).PS, VertexShader=@n29)) {name=LocalSpaceWPO, nodeColor=<rgba 0.1953 1 0 1>}
  n30 = VectorParameter() {parameterName=Direction, defaultValue=<rgba 0 1 0 1>}
  n31 = Transform(None=@n30.RGB) {transformSourceType=TRANSFORMSOURCE_Local}
  NamedRerouteDeclaration(None=Normalize(VectorInput=StaticSwitchParameter(True=@n31, False=@n30.RGB) {parameterName=UseLocalOrWorldProjection})) {name=DirectionVector, nodeColor=<rgba 1 0.7734 0 1>}
  NamedRerouteDeclaration(None=@n2.WorldPositionOffset) {name=WPO_original, nodeColor=<rgba 0.6875 0 1 1>}
  n32 = Divide(A=@n10, B=@n11)
  n33 = Saturate(None=Abs(None=@n32))
  n34 = Divide(A=ComponentMask(None=NamedRerouteUsage()) {b=true}, B=ScalarParameter() {defaultValue=242, parameterName=ScreenSize_Y})
  n35 = Saturate(None=Abs(None=@n34))
  n36 = Multiply(A=Ceil(None=OneMinus(None=@n33)), B=Ceil(None=OneMinus(None=@n35)))
  NamedRerouteDeclaration(None=@n36) {name=Final Opacity, nodeColor=<rgba 0 1 0.3438 1>}
  n37 = NamedRerouteUsage()
  n38 = Multiply(A=Time(), B=ScalarParameter() {defaultValue=0.1, parameterName=CamerMoveSpeed})
  n39 = Divide(A=Add(A=Sine(None=@n38)))
  n40 = AppendVector(
   A=Constant2Vector()
   B=LinearInterpolate(
    A=ScalarParameter() {defaultValue=-500, parameterName=CameraMove_Pos_X_1}
    B=ScalarParameter() {defaultValue=-200, parameterName=CameraMove_Pos_X_2}
    Alpha=@n39
   )
  )
  n41 = Subtract(A=Add(A=NamedRerouteUsage(), B=WorldPosition().XYZ), B=ObjectPositionWS())
  n42 = LinearInterpolate(
   A=ScalarParameter() {defaultValue=1.5, parameterName=CameraMove_Zoom_X_1}
   B=ScalarParameter() {defaultValue=0.35, parameterName=CameraMove_Zoom_X_2}
   Alpha=@n39
  )
  NamedRerouteDeclaration(None=Add(A=Add(A=@n37, B=@n40), B=Multiply(A=@n41, B=@n42))) {name=WPO Animation, nodeColor=<rgba 0.1719 1 0 1>}
  n43 = Add(A=@n32)
  n44 = Add(A=Multiply(A=@n34) {constB=-1})
  NamedRerouteDeclaration(None=AppendVector(A=Divide(A=@n43), B=Divide(A=@n44))) {name=ProcedrualScreenUVs, nodeColor=<rgba 1 0.5156 0 1>}
  NamedRerouteDeclaration(None=@n2.Normal) {name=Normal_Original, nodeColor=<rgba 0 0.9297 1 1>}
  ObjectPositionWS()
  Reroute(None=@n36)
  Saturate(None=LinearInterpolate(
    A=ScalarParameter() {parameterName=Fresnel_CharacterMask_Low}
    B=ScalarParameter() {defaultValue=0.184, parameterName=Fresnel_CharacterMask_High}
    Alpha=@n27
   ))
  SetMaterialAttributes!(MaterialAttributes, EmissiveColor=@n27) {attributeSetTypes [#4]}
  Sine(None=@n38)
 ]}
